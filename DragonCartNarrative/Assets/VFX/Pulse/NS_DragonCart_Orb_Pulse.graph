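# Niagara one-shot pulse effect for the DragonCart orb, combining an expanding refraction sphere, a warm flare with a light, and trailing rings.
# The sphere grows to User.PulseRadius over User.Duration, with camera-distance LOD on a material parameter.
# Five staggered ring meshes vary lifetime, size, opacity and X offset by partition; rings and flare follow System.OrbWPO in local space.
# A disabled GPU emitter places camera-centred sprites on the depth buffer, scaled by the animated radius.
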NiagaraSystem /DragonCartNarrative/Assets/VFX/Pulse/NS_DragonCart_Orb_Pulse
dependencies {
 usedDataInterfaces [
  &NiagaraDataInterfaceCurve
  &NiagaraDataInterfaceColorCurve
  &NiagaraDataInterfaceCamera
  &NiagaraDataInterfaceCollisionQuery
 ]
 usedDynamicInputs [
  &RandomRangeFloat
  &Multiply_VectorByFloat
  &self:ScratchDynamicInput_01
 ]
 usedModules [
  &SystemState
  &self:SystemSpawnScript.SystemScriptSource.SystemScriptGraph.NiagaraNodeAssignment_0.SetVariables_#1
  &self:SystemSpawnScript.SystemScriptSource.SystemScriptGraph.NiagaraNodeAssignment_1.SetVariables_#2
 ]
}
emitters {
 FlareLight {
  emitterUpdateScript [
   &EmitterState()
   &SpawnBurst_Instantaneous(Use Spawn Probability=false, UseLoopCountLimit=false)
  ]
  particleSpawnScript [
   &InitializeParticle(
    Color=<rgba 5 1.6446 0.2873 1>
    Lifetime=1.5
    Sprite Size Mode=::NewEnumerator3
    Uniform Sprite Size=&FN_LOD_Distance_Remap(Start Distance=10000, End Distance=20000, Minimum Output=5000, Maximum Output=6000, Clamp Output=true)
    UsePositionOffset=false
    Write Material Random=true
   )
   &CameraOffset(Camera Offset Amount=500)
   &SpriteFacingAndAlignment(WriteAlignment=false, WriteFacing=true)
  ]
  particleUpdateScript [
   &ParticleState()
   &ScaleColor(
    Scale Alpha=&FloatFromCurve(FloatCurve=DataInterface() {propertyValues {Curve {
        defaultValue=1
        keys [
         {arriveTangent=13.1896, interpMode=RCIM_Cubic, leaveTangent=13.1896, tangentMode=RCTM_User, time=0, value=0}
         {arriveTangent=-2.8238, interpMode=RCIM_Cubic, leaveTangent=-2.8238, tangentMode=RCTM_User, time=0.15, value=1}
         {tangentMode=RCTM_User, time=1, value=0}
        ]
       }}})
    ScaleA=true
    ScaleRGB=false
   )
   &DynamicMaterialParameters(
    Index 0 Param 1=0.7
    Index 0 Param 2=2
    Index 1 Param 1=700
    Index 2 Param 1=1
    Index 2 Param 2=&Lerp_Float(
     A=1.5
     B=0.25
     Alpha=&FloatFromCurve(FloatCurve=DataInterface() {propertyValues {Curve {
         defaultValue=1
         keys [
          {arriveTangent=1.7543, interpMode=RCIM_Cubic, leaveTangent=1.7543, tangentMode=RCTM_User, time=0, value=0}
          {arriveTangent=0.5144, tangentMode=RCTM_User, time=1, value=1}
         ]
        }}})
    )
    Module.Index0Float0Write=true
    Module.Index0Float1Write=true
    Module.Index0Float2Write=false
    Module.Index0Float3Write=false
    Module.Index1Float0Write=true
    Module.Index1Float1Write=false
    Module.Index1Float2Write=false
    Module.Index1Float3Write=false
    Module.Index2Float0Write=true
    Module.Index2Float1Write=true
    Module.Index2Float2Write=true
    Module.Index2Float3Write=true
    Write Parameter Index 1=::NewEnumerator1
    Write Parameter Index 2=::NewEnumerator1
   )
   &ScaleSpriteSize(Uniform Curve Sprite Scale=DataInterface() {propertyValues {Curve {keys [
        {arriveTangent=7.8051, interpMode=RCIM_Cubic, leaveTangent=7.8051, tangentMode=RCTM_User, time=0, value=0}
        {arriveTangent=0.5, interpMode=RCIM_Cubic, leaveTangent=0.5, time=0.2074, value=1}
        {tangentMode=RCTM_User, time=1, value=0.5}
       ]}}})
   /SetVariables(Particles.Position=&AddVectorToPosition(
     Position=<0 0 0>
     Vector=&TransformVector(Vector=$System.OrbWPO, Source Space=::World, Destination Space=::Local)
    )) {isSetParameters=true}
   /SetVariables(Particles.LightPosition:NiagaraPosition=&AddVectorToPosition(Position=$Particles.Position, Vector=<600 0 0>)) {isSetParameters=true}
   /SetVariables(Particles.LightColor=&Multiply_LinearColorByFloat(Linear Color=$Particles.Color, Float=5)) {isSetParameters=true}
   /SetVariables(Particles.LightRadius=&FloatFromCurve(
     FloatCurve=DataInterface() {propertyValues {Curve {
        defaultValue=1
        keys [
         {arriveTangent=1.8849, interpMode=RCIM_Cubic, leaveTangent=1.8849, tangentMode=RCTM_User, time=0, value=0.3}
         {arriveTangent=0.2703, tangentMode=RCTM_User, time=1, value=1}
        ]
       }}}
     Scale Curve=6400
    )) {isSetParameters=true}
  ]
  rendererClasses [&NiagaraSpriteRendererProperties, &NiagaraLightRendererProperties]
  renderers [
   {rendererClass=&NiagaraSpriteRendererProperties}
   {rendererClass=&NiagaraLightRendererProperties, rendererIndex=1}
   {rendererClass=&NiagaraSpriteRendererProperties, rendererIndex=2}
  ]
 }
 Minimal {
  bEnabled=false
  simTarget=GPUComputeSim
  emitterUpdateScript [
   &EmitterState()
   &SpawnRate(SpawnRate=1000, Use Spawn Probability=false)
  ]
  particleSpawnScript [
   &InitializeParticle(Color Mode=::NewEnumerator0, UsePositionOffset=false, Write Material Random=true)
   &ShapeLocation(
    MaskSpawnGroup=false
    Override Seed=false
    Shape Origin=&ReturnCameraPosition(
     Coordinate Space:ENiagaraCoordinateSpace=::Simulation
     Camera Query:NiagaraDataInterfaceCamera=DataInterface() {propertyValues {bRequireCurrentFrameData=false}}
    )
    Sphere Radius=5000
    Sphere Surface Distribution=1
    UseOwnerScale=true
   )
   /Niagara/Modules/Math/PlaceParticlesOnDepthBuffer_GPU(Max Z Depth:NiagaraFloat=$User.PulseRadius, Module.Use Max Z Depth=true)
  ]
  particleUpdateScript [
   &ParticleState()
   &ScaleSpriteSize(
    Uniform Curve Index=&RemapRange(
     Input Value To Remap=$System.PulseRadiusAnimated
     Input Min=0
     Input Max=&DistanceBetweenPositions(A=$Engine.Owner.Position, B=$Particles.Position)
     Output Min=0
     Output Max=1
    )
    Uniform Curve Sprite Scale=DataInterface() {propertyValues {Curve {
       defaultValue=0
       keys [
        {interpMode=RCIM_Cubic, tangentMode=RCTM_User, time=0, value=0}
        {interpMode=RCIM_Cubic, tangentMode=RCTM_User, time=0.5, value=1}
        {tangentMode=RCTM_User, time=1, value=0}
       ]
      }}}
   )
  ]
  rendererClasses [&NiagaraSpriteRendererProperties]
  renderers [{rendererClass=&NiagaraSpriteRendererProperties}]
 }
 RefractionSphere {
  emitterUpdateScript [
   &EmitterState()
   &SpawnBurst_Instantaneous(Spawn Time=0.03, Use Spawn Probability=false, UseLoopCountLimit=false)
  ]
  particleSpawnScript [
   &InitializeParticle(
    Color=<rgba 2 2 2 0.25>
    Lifetime=$User.Duration
    Mesh Scale Mode=::NewEnumerator3
    Mesh Uniform Scale=&Divide_Float(A=$User.PulseRadius, B=166.5)
    UsePositionOffset=false
    Write Material Random=true
   )
   &InitialMeshOrientation(Mesh Axis To Orient=<0 0 1>, Mesh Orientation Mode=::NewEnumerator2, Use Rotation Vector=false)
  ]
  particleUpdateScript [
   &ParticleState()
   &ScaleMeshSize(Scale Factor=&VectorFromFloat(Value=&FloatFromCurve(FloatCurve=DataInterface() {propertyValues {Curve {defaultValue=0, keys [{tangentMode=RCTM_User, time=0, value=0}, {tangentMode=RCTM_User, time=1, value=1}]}}})))
   &ScaleColor(
    Linear Color Curve=DataInterface() {propertyValues {
      AlphaCurve {keys [<at 0.4 1 RCIM_Cubic>, {arriveTangent=-2.2959, tangentMode=RCTM_User, time=1, value=0}]}
      BlueCurve {keys [<at 0 0 RCIM_Cubic>,<at 0.1594 0.0378>]}
      GreenCurve {keys [<at 0 0.8765 RCIM_Cubic>,<at 0.1594 0.3974>]}
      RedCurve {keys [<at 0 2 RCIM_Cubic>,<at 0.1594 1.2083>]}
     }}
    Scale Mode=::NewEnumerator2
   )
   &DynamicMaterialParameters(
    Index 0 Param 1=&Lerp_Float(
     A=1
     B=0.87
     Alpha=&FloatFromCurve(
      FloatCurve=DataInterface() {propertyValues {Curve {
         defaultValue=1
         keys [
          {interpMode=RCIM_Cubic, leaveTangentWeight=0.0333, tangentMode=RCTM_User, time=0, value=0}
          {
           arriveTangent=-1.6584
           arriveTangentWeight=0.0646
           interpMode=RCIM_Cubic
           leaveTangent=-1.6584
           leaveTangentWeight=0.581
           tangentMode=RCTM_User
           time=0.1
           value=1
          }
          {arriveTangentWeight=0.3, interpMode=RCIM_Cubic, tangentMode=RCTM_User, time=1, value=0}
         ]
        }}}
      Scale Curve=&FN_LOD_Distance_Remap(Start Distance=5000, End Distance=20000, Minimum Output=1, Maximum Output=0, Clamp Output=true)
     )
    )
    Index 0 Param 2=&FloatFromCurve(FloatCurve=DataInterface() {propertyValues {Curve {
        defaultValue=0
        keys [
         {arriveTangent=-2.4203, interpMode=RCIM_Cubic, time=0, value=1}
         {arriveTangent=-1.765, interpMode=RCIM_Cubic, leaveTangent=-1.765, tangentMode=RCTM_User, time=1, value=0}
        ]
       }}})
    Index 0 Param 3=&Lerp_Float(
     A=1
     B=0.5
     Alpha=&FloatFromCurve(FloatCurve=DataInterface() {propertyValues {Curve {defaultValue=1, keys [{arriveTangent=1.7704, interpMode=RCIM_Cubic, time=0, value=0}, {tangentMode=RCTM_User, time=1, value=1}]}}})
    )
    Index 0 Param 4=2000
    Module.Index0Float0Write=true
    Module.Index0Float1Write=true
    Module.Index0Float2Write=true
    Module.Index0Float3Write=true
   )
  ]
  rendererClasses [&NiagaraMeshRendererProperties]
  renderers [{rendererClass=&NiagaraMeshRendererProperties, rendererIndex=0}]
 }
 Rings {
  emitterUpdateScript [
   &EmitterState(Life Cycle Mode=::NewEnumerator1, Loop Behavior=::NewEnumerator2, Loop Count=5, Loop Duration=0.1, UseLoopDelay=false)
   &SpawnBurst_Instantaneous(Use Spawn Probability=false, UseLoopCountLimit=false)
  ]
  particleSpawnScript [
   &PartitionParticles(Number Of Partitions=5)
   &InitializeParticle(
    Color=<rgba 1 0.2614 0.0298 1>
    Lifetime=&Lerp_Float(A=2, B=1, Alpha=$Particles.PartitionParticles.NormalizedPartition)
    Mesh Scale Mode=::NewEnumerator3
    Mesh Uniform Scale=&Lerp_Float(A=8, B=12, Alpha=$Particles.PartitionParticles.NormalizedPartition)
    UsePositionOffset=false
    Write Material Random=true
   )
   &InitialMeshOrientation(Mesh Axis To Orient=<0 0 1>, Mesh Orientation Mode=::NewEnumerator2, Use Rotation Vector=false)
   &CameraOffset(Camera Offset Amount=100)
  ]
  particleUpdateScript [
   &ParticleState()
   /SetVariables(Particles.Position=&AddVectorToPosition(
     Position=&MakeVector(
      X=&FloatFromCurve(
       FloatCurve=DataInterface() {propertyValues {Curve {
          defaultValue=0
          keys [{arriveTangent=1.2988, interpMode=RCIM_Cubic, leaveTangent=1.2988, tangentMode=RCTM_User, time=0, value=0.25}, <at 1 1 RCIM_Cubic>]
         }}}
       Scale Curve=&Multiply_FloatByInt(Float=100, Integer=$Particles.PartitionParticles.Partition)
      )
      Y=0
      Z=0
     )
     Vector=&TransformVector(Vector=$System.OrbWPO, Source Space=::World, Destination Space=::Local)
    )) {isSetParameters=true}
   &ScaleColor(
    Scale Alpha=&FloatFromCurve(
     FloatCurve=DataInterface() {propertyValues {Curve {
        defaultValue=0
        keys [
         {arriveTangent=18.046, interpMode=RCIM_Cubic, leaveTangent=18.046, tangentMode=RCTM_User, time=0, value=0}
         {arriveTangent=-2.3141, interpMode=RCIM_Cubic, leaveTangent=-2.3141, tangentMode=RCTM_User, time=0.1, value=1}
         <at 1 0 RCIM_Cubic>
        ]
       }}}
     Scale Curve=&Lerp_Float(A=1, B=0.4, Alpha=$Particles.PartitionParticles.NormalizedPartition)
    )
    ScaleA=true
    ScaleRGB=false
   )
   &ScaleMeshSize(Scale Factor=&VectorFromFloat(Value=&FloatFromCurve(
      FloatCurve=DataInterface() {propertyValues {Curve {
         defaultValue=1
         keys [
          {arriveTangent=1.6468, interpMode=RCIM_Cubic, leaveTangent=1.6468, tangentMode=RCTM_User, time=0, value=0.25}
          {arriveTangent=0.2249, tangentMode=RCTM_User, time=1, value=1}
         ]
        }}}
      Scale Curve=&Lerp_Float(A=1, B=2, Alpha=$Particles.PartitionParticles.NormalizedPartition)
     )))
   &DynamicMaterialParameters(
    Index 0 Param 1=&FloatFromCurve(
     FloatCurve=DataInterface() {propertyValues {Curve {keys [{arriveTangent=-1.6926, interpMode=RCIM_Cubic, time=0, value=1}, <at 1 0.1 RCIM_Cubic>]}}}
     Scale Curve=500
    )
    Index 0 Param 2=&FloatFromCurve(FloatCurve=DataInterface() {propertyValues {Curve {defaultValue=1, keys [<at 0.25 0 RCIM_Cubic>,<at 1 1 RCIM_Cubic>]}}}, Scale Curve=6)
    Index 0 Param 3=1.5
    Module.Index0Float0Write=true
    Module.Index0Float1Write=true
    Module.Index0Float2Write=true
    Module.Index0Float3Write=false
   )
  ]
  rendererClasses [&NiagaraMeshRendererProperties]
  renderers [{rendererClass=&NiagaraMeshRendererProperties, rendererIndex=0}]
 }
}
summary {
 systemName=NS_DragonCart_Orb_Pulse
 emitters [
  {emitterName=RefractionSphere, rendererClasses [&NiagaraMeshRendererProperties]}
  {emitterName=FlareLight, rendererClasses [&NiagaraSpriteRendererProperties, &NiagaraLightRendererProperties]}
  {emitterName=Rings, rendererClasses [&NiagaraMeshRendererProperties]}
  {bEnabled=false, emitterName=Minimal, simTarget=GPUComputeSim, rendererClasses [&NiagaraSpriteRendererProperties]}
 ]
 userVariables [
  User.Duration {type=NiagaraFloat, defaultValue=2.5, description=""}
  User.PulseRadius {type=NiagaraFloat, defaultValue=15000, description=""}
 ]
}
